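# Animation Blueprint for the SharpMagnet backpack cosmetic, animating the pack in response to player events.
# A state machine leaves looping idle for skydive, DBNO, pain, elimination, chest-open and notify-randomized idle breaks, driven by anim input events and flags.
# Its pose feeds spring-based Anim Dynamics on the base bone, with alpha set by skydiving.
# The noise-driven neck transform, testing state machine and health-comparison chain are left unconnected.
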AnimBlueprint /BRCosmetics/Accessories/FORT_Backpacks/Backpack_M_MED_SharpMagnet/Meshes/M_MED_SharpMagnet_Pack_AnimBP
target_skeleton=&/M_MED_SharpMagnet_Pack_Skeleton
graphs [
 AnimGraph {nodes [
   Input Pose()
   "Transform (Modify) Bone - Bone: neck_01"(Alpha=0.7, Translation=Get movement() -> (movement:Vector), Rotation="0,0,0", Scale="1,1,1") {
    rotationSpace=BCS_WorldSpace
    translationMode=BMM_Additive
    translationSpace=BCS_BoneSpace
    alphaBoolBlend {blendOption=Linear}
    alphaScaleBiasClamp {bClampResult=true, clampMax=0.7, interpSpeedDecreasing=2, interpSpeedIncreasing=2}
    boneToModify {boneName=neck_01}
   }
   AnimGraphNode_StateMachine_1 = Backpack States() {is=title AnimGraphNode_StateMachine}
   n1 = Anim Dynamics - Bone: base(ComponentPose=Local To Component(LocalPose=@AnimGraphNode_StateMachine_1), Alpha=Get Skydive() -> (Skydive:double), ExternalForce="0,0,0") {
    angularDampingOverride=0.9
    angularSpringConstant=200000
    bAngularSpring=true
    bGravityOverrideInSimSpace=true
    bLinearSpring=true
    bOverrideAngularDamping=true
    bOverrideLinearDamping=true
    gravityScale=0
    linearDampingOverride=0.995
    linearSpringConstant=50
    simulationSpace=BoneRelative
    alphaBoolBlend {blendOption=Linear}
    alphaScaleBiasClamp {bClampResult=true, clampMax=0.2}
    boundBone {boneName=base}
    componentAppliedLinearAccClamp=<6000 6000 6000>
    componentLinearAccScale=<0.3 0.3 0.1>
    componentLinearVelScale=<0.7 0.7 0.8>
    gravityOverride {z=-980}
    physicsBodyDefinitions [{
      sphereCollisionRadius=0
      boundBone {boneName=base}
      boxExtents=<5 40 40>
      constraintSetup {
       angularTargetAxis=AxisZ
       twistAxis=AxisZ
       angularLimitsMax=<0 0 0>
       angularLimitsMin=<0 0 0>
       angularTarget=<0 0 0>
       linearAxesMax=<3 10 12>
       linearAxesMin=<-3 -8 -11>
      }
      localJointOffset=<-2 0 0>
     }]
    relativeSpaceBone {boneName=root}
   }
   Output Pose(Result=Component To Local(ComponentPose=@n1)) {layerGroup=None}
   AnimGraphNode_StateMachine_0 = Animation Testing() {is=title AnimGraphNode_StateMachine}
  ]}
 Animation Testing {
  entry=State
  path=AnimGraph.AnimGraphNode_StateMachine_0.Animation Testing
  ids=id
  nodes [
   AnimStateNode_0 = State() {is=title AnimStateNode}
   AnimStateNode_1 = State_5() {is=title AnimStateNode}
   AnimStateNode_2 = State_1() {is=title AnimStateNode}
   AnimStateNode_3 = State_2() {is=title AnimStateNode}
   AnimStateNode_4 = State_3() {is=title AnimStateNode}
   AnimStateNode_5 = State_4() {is=title AnimStateNode}
   AnimStateTransitionNode_0 = AnimStateTransitionNode() {bAutomaticRuleBasedOnSequencePlayerInState=true, crossfadeDuration=0}
   AnimStateTransitionNode_2 = AnimStateTransitionNode()
   AnimStateTransitionNode_3 = AnimStateTransitionNode()
   AnimStateTransitionNode_4 = AnimStateTransitionNode()
   AnimStateTransitionNode_5 = AnimStateTransitionNode()
   AnimStateTransitionNode_6 = AnimStateTransitionNode()
  ]
  links [
   {
    dst=State_1
    rule=AnimGraph.AnimGraphNode_StateMachine_0.Animation Testing.AnimStateTransitionNode_0.Transition
    src=State
    via=AnimStateTransitionNode_0
   }
   {
    dst=State_2
    rule=AnimGraph.AnimGraphNode_StateMachine_0.Animation Testing.AnimStateTransitionNode_2.Transition
    src=State_1
    via=AnimStateTransitionNode_2
   }
   {
    dst=State_3
    rule=AnimGraph.AnimGraphNode_StateMachine_0.Animation Testing.AnimStateTransitionNode_3.Transition
    src=State_2
    via=AnimStateTransitionNode_3
   }
   {
    dst=State_4
    rule=AnimGraph.AnimGraphNode_StateMachine_0.Animation Testing.AnimStateTransitionNode_4.Transition
    src=State_3
    via=AnimStateTransitionNode_4
   }
   {
    dst=State_5
    rule=AnimGraph.AnimGraphNode_StateMachine_0.Animation Testing.AnimStateTransitionNode_5.Transition
    src=State_4
    via=AnimStateTransitionNode_5
   }
   {
    dst=State
    rule=AnimGraph.AnimGraphNode_StateMachine_0.Animation Testing.AnimStateTransitionNode_6.Transition
    src=State_5
    via=AnimStateTransitionNode_6
   }
  ]
 }
 State {
  path=AnimGraph.AnimGraphNode_StateMachine_0.Animation Testing.AnimStateNode_0.State
  nodes [
   n1 = Sequence Player 'SharpMagnet_Backpack_Idle'() {bLoopAnimation=false, sequence=/BRCosmetics/Accessories/FORT_Backpacks/Backpack_M_MED_SharpMagnet/Animation/SharpMagnet_Backpack_Idle}
   Output Animation Pose(Result=@n1)
  ]
 }
 State_5 {
  path=AnimGraph.AnimGraphNode_StateMachine_0.Animation Testing.AnimStateNode_1.State_5
  nodes [
   n1 = Sequence Player 'SharpMagnet_Backpack_DBNO'() {bLoopAnimation=false, sequence=/BRCosmetics/Accessories/FORT_Backpacks/Backpack_M_MED_SharpMagnet/Animation/SharpMagnet_Backpack_DBNO}
   Output Animation Pose(Result=@n1)
  ]
 }
 State_1 {
  path=AnimGraph.AnimGraphNode_StateMachine_0.Animation Testing.AnimStateNode_2.State_1
  nodes [
   n1 = Sequence Player 'SharpMagnet_Backpack_Idle'() {bLoopAnimation=false, sequence=&SharpMagnet_Backpack_Idle}
   Output Animation Pose(Result=@n1)
  ]
 }
 State_2 {
  path=AnimGraph.AnimGraphNode_StateMachine_0.Animation Testing.AnimStateNode_3.State_2
  nodes [
   n1 = Sequence Player 'SharpMagnet_Backpack_IdleBreak01'() {bLoopAnimation=false, sequence=/BRCosmetics/Accessories/FORT_Backpacks/Backpack_M_MED_SharpMagnet/Animation/SharpMagnet_Backpack_IdleBreak01}
   Output Animation Pose(Result=@n1)
  ]
 }
 State_3 {
  path=AnimGraph.AnimGraphNode_StateMachine_0.Animation Testing.AnimStateNode_4.State_3
  nodes [
   n1 = Sequence Player 'SharpMagnet_Backpack_OpenChest'() {bLoopAnimation=false, sequence=/BRCosmetics/Accessories/FORT_Backpacks/Backpack_M_MED_SharpMagnet/Animation/SharpMagnet_Backpack_OpenChest}
   Output Animation Pose(Result=@n1)
  ]
 }
 State_4 {
  path=AnimGraph.AnimGraphNode_StateMachine_0.Animation Testing.AnimStateNode_5.State_4
  nodes [
   n1 = Sequence Player 'SharpMagnet_Backpack_Elimination'() {bLoopAnimation=false, sequence=/BRCosmetics/Accessories/FORT_Backpacks/Backpack_M_MED_SharpMagnet/Animation/SharpMagnet_Backpack_Elimination}
   Output Animation Pose(Result=@n1)
  ]
 }
 Transition {path=AnimGraph.AnimGraphNode_StateMachine_0.Animation Testing.AnimStateTransitionNode_0.Transition, nodes [Result()]}
 Transition {path=AnimGraph.AnimGraphNode_StateMachine_0.Animation Testing.AnimStateTransitionNode_2.Transition, nodes [Result()]}
 Transition {path=AnimGraph.AnimGraphNode_StateMachine_0.Animation Testing.AnimStateTransitionNode_3.Transition, nodes [Result()]}
 Transition {path=AnimGraph.AnimGraphNode_StateMachine_0.Animation Testing.AnimStateTransitionNode_4.Transition, nodes [Result()]}
 Transition {path=AnimGraph.AnimGraphNode_StateMachine_0.Animation Testing.AnimStateTransitionNode_5.Transition, nodes [Result()]}
 Transition {path=AnimGraph.AnimGraphNode_StateMachine_0.Animation Testing.AnimStateTransitionNode_6.Transition, nodes [Result()]}
 Backpack States {
  entry=Locomotion
  path=AnimGraph.AnimGraphNode_StateMachine_1.Backpack States
  ids=id
  nodes [
   AnimStateNode_0 = Locomotion() {is=title AnimStateNode, stateEntered {guid=#1, notifyName=Entered_RandomLocoBreak}}
   AnimStateNode_1 = Pain() {is=title AnimStateNode}
   AnimStateNode_2 = Skydiving() {
    is=title AnimStateNode
    stateEntered {guid=#2}
    stateFullyBlended {guid=#3}
    stateLeft {guid=#4}
   }
   AnimStateNode_3 = DBNO() {is=title AnimStateNode}
   AnimStateNode_4 = Elimination() {is=title AnimStateNode}
   AnimStateNode_5 = OpenChest() {is=title AnimStateNode}
   AnimStateNode_6 = IdleBreak() {is=title AnimStateNode}
   AnimStateTransitionNode_10 = AnimStateTransitionNode()
   AnimStateTransitionNode_11 = AnimStateTransitionNode()
   AnimStateTransitionNode_12 = AnimStateTransitionNode() {crossfadeDuration=0.05}
   AnimStateTransitionNode_13 = AnimStateTransitionNode() {bAutomaticRuleBasedOnSequencePlayerInState=true, crossfadeDuration=0.05}
   AnimStateTransitionNode_2 = AnimStateTransitionNode()
   AnimStateTransitionNode_3 = AnimStateTransitionNode()
   AnimStateTransitionNode_4 = AnimStateTransitionNode()
   AnimStateTransitionNode_5 = AnimStateTransitionNode() {bAutomaticRuleBasedOnSequencePlayerInState=true}
   AnimStateTransitionNode_6 = AnimStateTransitionNode()
   AnimStateTransitionNode_7 = AnimStateTransitionNode()
   AnimStateTransitionNode_8 = AnimStateTransitionNode() {bAutomaticRuleBasedOnSequencePlayerInState=true}
   AnimStateTransitionNode_9 = AnimStateTransitionNode() {bAutomaticRuleBasedOnSequencePlayerInState=true}
  ]
  links [
   {
    dst=Locomotion
    rule=AnimGraph.AnimGraphNode_StateMachine_1.Backpack States.AnimStateTransitionNode_4.Transition
    src=DBNO
    via=AnimStateTransitionNode_4
   }
   {
    dst=Locomotion
    rule=AnimGraph.AnimGraphNode_StateMachine_1.Backpack States.AnimStateTransitionNode_5.Transition
    src=Elimination
    via=AnimStateTransitionNode_5
   }
   {
    dst=Locomotion
    rule=AnimGraph.AnimGraphNode_StateMachine_1.Backpack States.AnimStateTransitionNode_9.Transition
    src=IdleBreak
    via=AnimStateTransitionNode_9
   }
   {
    dst=DBNO
    rule=AnimGraph.AnimGraphNode_StateMachine_1.Backpack States.AnimStateTransitionNode_3.Transition
    src=Locomotion
    via=AnimStateTransitionNode_3
   }
   {
    dst=Elimination
    rule=AnimGraph.AnimGraphNode_StateMachine_1.Backpack States.AnimStateTransitionNode_11.Transition
    src=Locomotion
    via=AnimStateTransitionNode_11
   }
   {
    dst=IdleBreak
    rule=AnimGraph.AnimGraphNode_StateMachine_1.Backpack States.AnimStateTransitionNode_6.Transition
    src=Locomotion
    via=AnimStateTransitionNode_6
   }
   {
    dst=OpenChest
    rule=AnimGraph.AnimGraphNode_StateMachine_1.Backpack States.AnimStateTransitionNode_7.Transition
    src=Locomotion
    via=AnimStateTransitionNode_7
   }
   {
    dst=Pain
    rule=AnimGraph.AnimGraphNode_StateMachine_1.Backpack States.AnimStateTransitionNode_12.Transition
    src=Locomotion
    via=AnimStateTransitionNode_12
   }
   {
    dst=Skydiving
    rule=AnimGraph.AnimGraphNode_StateMachine_1.Backpack States.AnimStateTransitionNode_10.Transition
    src=Locomotion
    via=AnimStateTransitionNode_10
   }
   {
    dst=Locomotion
    rule=AnimGraph.AnimGraphNode_StateMachine_1.Backpack States.AnimStateTransitionNode_8.Transition
    src=OpenChest
    via=AnimStateTransitionNode_8
   }
   {
    dst=Locomotion
    rule=AnimGraph.AnimGraphNode_StateMachine_1.Backpack States.AnimStateTransitionNode_13.Transition
    src=Pain
    via=AnimStateTransitionNode_13
   }
   {
    dst=Locomotion
    rule=AnimGraph.AnimGraphNode_StateMachine_1.Backpack States.AnimStateTransitionNode_2.Transition
    src=Skydiving
    via=AnimStateTransitionNode_2
   }
  ]
 }
 Locomotion {
  path=AnimGraph.AnimGraphNode_StateMachine_1.Backpack States.AnimStateNode_0.Locomotion
  nodes [Output Animation Pose(Result=Sequence Player 'SharpMagnet_Backpack_Idle'() {sequence=&SharpMagnet_Backpack_Idle})]
 }
 Pain {
  path=AnimGraph.AnimGraphNode_StateMachine_1.Backpack States.AnimStateNode_1.Pain
  nodes [
   Sequence Player 'SharpMagnet_Backpack_Pain'() {bLoopAnimation=false, sequence=/BRCosmetics/Accessories/FORT_Backpacks/Backpack_M_MED_SharpMagnet/Animation/SharpMagnet_Backpack_Pain}
   n1 = Sequence Player 'SharpMagnet_Backpack_Pain'() {bLoopAnimation=false, sequence=&SharpMagnet_Backpack_Pain}
   Output Animation Pose(Result=@n1)
  ]
 }
 Skydiving {
  path=AnimGraph.AnimGraphNode_StateMachine_1.Backpack States.AnimStateNode_2.Skydiving
  nodes [Output Animation Pose(Result=Sequence Player 'SharpMagnet_Backpack_SkyDiving'() {sequence=/BRCosmetics/Accessories/FORT_Backpacks/Backpack_M_MED_SharpMagnet/Animation/SharpMagnet_Backpack_SkyDiving})]
 }
 DBNO {
  path=AnimGraph.AnimGraphNode_StateMachine_1.Backpack States.AnimStateNode_3.DBNO
  nodes [Output Animation Pose(Result=Sequence Player 'SharpMagnet_Backpack_DBNO'() {sequence=&SharpMagnet_Backpack_DBNO})]
 }
 Elimination {
  path=AnimGraph.AnimGraphNode_StateMachine_1.Backpack States.AnimStateNode_4.Elimination
  nodes [Output Animation Pose(Result=Sequence Player 'SharpMagnet_Backpack_Elimination'() {sequence=&SharpMagnet_Backpack_Elimination})]
 }
 OpenChest {
  path=AnimGraph.AnimGraphNode_StateMachine_1.Backpack States.AnimStateNode_5.OpenChest
  nodes [Output Animation Pose(Result=Sequence Player 'SharpMagnet_Backpack_OpenChest'() {sequence=&SharpMagnet_Backpack_OpenChest})]
 }
 IdleBreak {
  path=AnimGraph.AnimGraphNode_StateMachine_1.Backpack States.AnimStateNode_6.IdleBreak
  nodes [
   n1 = Sequence Player 'SharpMagnet_Backpack_IdleBreak01'!(bLoopAnimation:bool=False) -> (Pose) {bLoopAnimation=false, sequence=&SharpMagnet_Backpack_IdleBreak01}
   n2 = Sequence Player 'SharpMagnet_Backpack_IdleBreak01'(bLoopAnimation=False) {bLoopAnimation=false, sequence=&SharpMagnet_Backpack_IdleBreak01}
   n3 = Sequence Player 'SharpMagnet_Backpack_IdleBreak01'(bLoopAnimation=False) {bLoopAnimation=false, sequence=&SharpMagnet_Backpack_IdleBreak01}
   n4 = Blend Poses by int!(
    BlendPose_0:Pose Link Structure=@n2
    BlendPose_1:Pose Link Structure=@n1
    BlendPose_2:Pose Link Structure=@n3
    BlendTime_0:float=0.1
    BlendTime_1:float=0.1
    BlendTime_2:float=0.1
    ActiveChildIndex:int=Get Choose_LocoBreak_Index() -> (Choose_LocoBreak_Index:int=0)
   ) -> (Pose:Pose Link Structure) {
    is=title AnimGraphNode_BlendListByInt
    blendType=Linear
    blendPose [{linkId=-1, sourceLinkId=-1}, {linkId=-1, sourceLinkId=-1}, {linkId=-1, sourceLinkId=-1}]
    blendTime [0.1,0.1,0.1]
   }
   Output Animation Pose(Result=@n4)
  ]
 }
 Transition {
  path=AnimGraph.AnimGraphNode_StateMachine_1.Backpack States.AnimStateTransitionNode_10.Transition
  nodes [Result(bCanEnterTransition=Get bIsSkyDiving())]
 }
 Transition {
  path=AnimGraph.AnimGraphNode_StateMachine_1.Backpack States.AnimStateTransitionNode_11.Transition
  nodes [Result(bCanEnterTransition=Get bEnemyPlayerEliminated() -> (bEnemyPlayerEliminated:bool))]
 }
 Transition {
  path=AnimGraph.AnimGraphNode_StateMachine_1.Backpack States.AnimStateTransitionNode_12.Transition
  nodes [Result(bCanEnterTransition=Get TookDamage() -> (TookDamage:bool))]
 }
 Transition {path=AnimGraph.AnimGraphNode_StateMachine_1.Backpack States.AnimStateTransitionNode_13.Transition, nodes [Result()]}
 Transition {
  path=AnimGraph.AnimGraphNode_StateMachine_1.Backpack States.AnimStateTransitionNode_2.Transition
  nodes [Result(bCanEnterTransition=NOT Boolean(A=Get bIsSkyDiving()))]
 }
 Transition {
  path=AnimGraph.AnimGraphNode_StateMachine_1.Backpack States.AnimStateTransitionNode_3.Transition
  nodes [Result(bCanEnterTransition=Get bIsDBNO())]
 }
 Transition {
  path=AnimGraph.AnimGraphNode_StateMachine_1.Backpack States.AnimStateTransitionNode_4.Transition
  nodes [Result(bCanEnterTransition=NOT Boolean(A=Get bIsDBNO()))]
 }
 Transition {path=AnimGraph.AnimGraphNode_StateMachine_1.Backpack States.AnimStateTransitionNode_5.Transition, nodes [Result()]}
 Transition {
  path=AnimGraph.AnimGraphNode_StateMachine_1.Backpack States.AnimStateTransitionNode_6.Transition
  nodes [Result(bCanEnterTransition=Get AllowLocomotionBreak() -> (AllowLocomotionBreak:bool))]
 }
 Transition {
  path=AnimGraph.AnimGraphNode_StateMachine_1.Backpack States.AnimStateTransitionNode_7.Transition
  nodes [Result(bCanEnterTransition=Get bChestOpened() -> (bChestOpened:bool))]
 }
 Transition {path=AnimGraph.AnimGraphNode_StateMachine_1.Backpack States.AnimStateTransitionNode_8.Transition, nodes [Result()]}
 Transition {
  path=AnimGraph.AnimGraphNode_StateMachine_1.Backpack States.AnimStateTransitionNode_9.Transition
  nodes [
   Result(bCanEnterTransition=NOT Boolean(A=Get AllowLocomotionBreak()))
   Get IsLocoBreaking() -> (IsLocoBreaking:bool)
  ]
 }
 EventGraph {nodes [
   n1 = Cast To FortPlayerPawn(Object=GetOwningActor())
   n2 = GetHealth!(self:Fort Pawn Object Reference=@n1) -> (ReturnValue:float)
   > Set Current Health(Current Health:double=@n2) -> (Output_Get:double)
   > n3 = Branch(Condition="Not Equal (Float)"(A=@n2, B=Get Previous Health() -> (Previous Health:double)))
   > Set TookDamage(TookDamage:bool=true) -> (Output_Get:bool)
   > Set Previous Health(Previous Health:double=Get Current Health() -> (Current Health:double)) -> (Output_Get:double)
   Set TookDamage(execute=@n3.else, TookDamage=false)
   Event BlueprintUpdateAnimation()
   > n4 = Sequence()
   > n5 = Random Vector Function!(
    execute
    CharacterPartAnimInstance:Custom Character Part Anim Instance Object Reference
    SkelMeshComponent:Mesh Component Object Reference=GetOwnerSkeletalMeshComponent()
    JointName:Name=neck_01
    GameTimeInSeconds:double=GetGameTimeInSeconds()
    TimeOffset:double=0
    PerlinScale:double=0
    Frequency:Vector="0.9,1.1,0.75"
    Amplitude:Vector="0.3,0.7,0.5"
    Offset:Vector="0,0,0"
   ) -> (then, VectorOut:Vector)
   > Set movement(movement:Vector=@n5) -> (Output_Get:Vector)
   Set Skydive(execute=@n4.then_1, Skydive:double=SelectFloat(A=0.2, B=1, bPickA=Get bIsSkyDiving())) -> (Output_Get:double)
   AnimNotify_TriggerLocomotionRandomizer() {is=title K2Node_Event}
   > n6 = Set AllowLocomotionBreak(AllowLocomotionBreak:bool=RandomBoolWithWeight!(Weight:float=0.3) -> (ReturnValue:bool)) -> (Output_Get:bool)
   > Branch(Condition=@n6)
   > Set Choose_LocoBreak_Index(Choose_LocoBreak_Index:int=RandomInteger!(Max:int=3) -> (ReturnValue:int=0)) -> (Output_Get:int=0)
   Event BlueprintPostEvaluateAnimation() {is=title K2Node_Event}
   > Set bEnemyPlayerEliminated(bEnemyPlayerEliminated:bool=false) -> (Output_Get:bool)
   > Set bChestOpened(bChestOpened:bool=false) -> (Output_Get:bool)
   AnimNotify_Entered_RandomLocoBreak() {is=title K2Node_Event}
   > Set AllowLocomotionBreak(AllowLocomotionBreak=false)
   AnimNotify_IsLocoBreaking() {is=title K2Node_Event}
   > Set IsLocoBreaking(IsLocoBreaking:bool=false) -> (Output_Get:bool)
   n7 = Event BlueprintProcessAnimInputEvent()
   > n8 = Sequence()
   > n9 = Branch(Condition=IsNamed(self=@n7.AnimInputEvent, InEventName=aie.EnemyPlayerEliminated))
   > Set bEnemyPlayerEliminated(bEnemyPlayerEliminated=true)
   Branch(execute=@n9.else, Condition=IsNamed(self=@n7.AnimInputEvent, InEventName=aie.ChestOpened))
   > Set bChestOpened(bChestOpened=true)
   Branch(
    execute=@n8.then_1
    Condition=OR Boolean(
     A=IsNamed(self=@n7.AnimInputEvent, InEventName=aie.PlayerVitality.HealthDecreased)
     B=IsNamed(self=@n7.AnimInputEvent, InEventName=Noneaie.PlayerVitality.ShieldsDecreased)
    )
   )
   > Set TookDamage(TookDamage=true)
   > RetriggerableDelay(Duration=0.2)
   > Set TookDamage(TookDamage=false)
   TryGetPawnOwner()
  ]}
]
variables [
 AllowLocomotionBreak {type=boolean}
 bChestOpened {type=boolean}
 bEnemyPlayerEliminated {type=boolean}
 Choose_LocoBreak_Index {type=integer}
 Current Health {type=number}
 IsLocoBreaking {type=boolean}
 movement {type=Vector}
 Previous Health {type=number}
 Skydive {type=number}
 TookDamage {type=boolean}
]
